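# Info row widget for a matchmaking settings entry, showing a setting icon beside a wrapping name label.
# A horizontal box holds the material-driven icon and the CommonTextBlock that fills the remaining width.
# PreConstruct calls SetMobileStyle, which enlarges the icon's desired size when IsUsingCompactUI is true.
# Anim_OnFocus briefly fades the name from grey to white; this widget's own graphs never play it.
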WidgetBlueprint /MatchmakingSettingsUI/MatchmakingSettings/SettingEntries/WBP_MatchmakingSettings_SettingEntryInfo
animations [Anim_OnFocus {
  binding_count=1
  length_seconds=0.2
  object=&self:Anim_OnFocus
  sequence {bindings [Txt_SettingName {
     id=#1
     possessed_class=CommonTextBlock
     tracks [{
       type=MovieSceneColorTrack
       display_name=Color and Opacity
       sections [{
         type=MovieSceneColorSection
         end_frame=4
         start_frame=0
         channels [
          Color.R {type=MovieSceneScriptingFloatChannel, keys [[0,0.5776],[4,1]]}
          Color.G {type=MovieSceneScriptingFloatChannel, keys [[0,0.5776],[4,1]]}
          Color.B {type=MovieSceneScriptingFloatChannel, keys [[0,0.5776],[4,1]]}
          Color.A {type=MovieSceneScriptingFloatChannel, const [1,2]}
         ]
        }]
      }]
    }]}
 }]
graphs [
 EventGraph {nodes [
   Event PreConstruct()
   > SetMobileStyle() {is=title K2Node_CallFunction}
  ]}
 SetMobileStyle {nodes [
   SetMobileStyle() {is=title K2Node_FunctionEntry}
   > SetDesiredSizeOverride!(
    execute
    self:Image Object Reference=Get Img_SettingIcon() -> (Img_SettingIcon:Image Object Reference)
    DesiredSize:Vector 2D Structure=Select(Option 0:Vector 2D Structure=(X=40,Y=40), Option 1:Vector 2D Structure=(X=56,Y=56), Index=IsUsingCompactUI()) -> (ReturnValue:Vector 2D Structure)
   ) -> (then)
  ]}
]
variables []
widgets {
 info {inheritedWidgetCount=0, parentClass=&UserWidget, rootWidgetClass=&HorizontalBox, widgetCount=3}
 widgets [
  HB_SettingInfo {widgetClassPath=&HorizontalBox, widget=HB_SettingInfo}
  Img_SettingIcon {
   widgetClassPath=&Image
   bIsVariable=true
   parent=HB_SettingInfo
   widget=Img_SettingIcon
   properties {
    visibility=SelfHitTestInvisible
    brush {
     bIsDynamicallyLoaded=false
     drawAs=Image
     imageType=FullColor
     mirroring=NoMirror
     resourceName=None
     resourceObject=/MatchmakingSettingsUI/MatchmakingSettings/MaterialInstances/MI_UI_MatchmakingSettings_Icon
     tiling=NoTile
     imageSize=<40 40>
     margin=<ltrb 0 0 0 0>
     outlineSettings {
      bUseBrushTransparency=false
      roundingType=HalfHeightRadius
      width=0
      color {colorUseRule=UseColor_Specified, specifiedColor=<rgba 0 0 0 0>}
      cornerRadii=<0 0 0 0>
     }
     tintColor {colorUseRule=UseColor_Specified, specifiedColor=<rgba 1 1 1 1>}
     uVRegion {bIsValid=false, max=<0 0>, min=<0 0>}
    }
   }
   slot {class=HorizontalBoxSlot, properties {horizontalAlignment=HAlign_Left, verticalAlignment=VAlign_Center, padding=<ltrb 0 0 12 0>}}
  }
  Txt_SettingName {
   widgetClassPath=&CommonTextBlock
   bIsVariable=true
   parent=HB_SettingInfo
   widget=Txt_SettingName
   properties {
    autoWrapText=true
    scrollStyle=&TextScrollStyle-Base_C
    text=$SettingName
    colorAndOpacity {colorUseRule=UseColor_Specified, specifiedColor=<rgba 0.5776 0.5776 0.5776 1>}
    font {
     bForceMonospaced=false
     bMaterialIsStencil=false
     fontMaterial=null
     fontObject=null
     letterSpacing=0
     monospacedWidth=1
     size=24
     skewAmount=0
     typefaceFontName=None
     outlineSettings {
      bApplyOutlineToDropShadows=false
      bMiteredCorners=false
      bSeparateFillAlpha=false
      outlineMaterial=null
      outlineSize=0
      outlineColor=<rgba 0 0 0 1>
     }
    }
    shadowOffset=<0 0>
   }
   slot {class=HorizontalBoxSlot, properties {verticalAlignment=VAlign_Center, size {sizeRule=Fill, value=1}}}
  }
 ]
}
