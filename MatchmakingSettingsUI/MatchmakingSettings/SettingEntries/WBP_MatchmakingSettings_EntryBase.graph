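# Base button widget for a matchmaking settings entry row, built on UIKitModularButton, with an info panel and a named slot for the setting's control.
# Hover animates a focus outline and plays the info panel's focus animation, reversing it on unhover.
# SetDisableState drives an IsDisabled parameter on the dynamic background material and dims both info and slot.
# PreConstruct sets a taller minimum row height when the compact UI is in use.
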WidgetBlueprint /MatchmakingSettingsUI/MatchmakingSettings/SettingEntries/WBP_MatchmakingSettings_EntryBase
graphs [
 EventGraph {nodes [
   Event PreConstruct()
   > SetMobileStyle() {is=title K2Node_CallFunction}
   Event OnInitialized()
   > InitVisuals()
   Event On Hovered()
   n1 = Get WBP_UIKit_Block_Outline() -> (WBP_UIKit_Block_Outline:WBP UIKit Block Outline Object Reference)
   > Focused_Visualization!(execute, self:WBP UIKit Block Outline Object Reference=@n1, NeedsToAnimate:bool=true) -> (then)
   n2 = Get SettingEntryInfo() -> (SettingEntryInfo:WBP Matchmaking Settings Setting Entry Info Object Reference)
   n3 = Get Anim_OnFocus!(self:WBP Matchmaking Settings Setting Entry Info Object Reference=@n2) -> (Anim_OnFocus:Widget Animation Object Reference)
   > PlayAnimationForward(self=@n2, InAnimation=@n3)
   Event On Unhovered()
   > Unfocused_Visualization!(execute, self:WBP UIKit Block Outline Object Reference=@n1, NeedsToAnimate:bool=true) -> (then)
   > PlayAnimationReverse(self=@n2, InAnimation=@n3)
  ]}
 InitVisuals {nodes [
   InitVisuals() {is=title K2Node_FunctionEntry}
   > n1 = GetDynamicMaterial(self:Border Object Reference=Get B_Setting() -> (B_Setting:Common Border Object Reference))
   > Set DynamicBackground(DynamicBackground:Material Instance Dynamic Object Reference=@n1) -> (Output_Get:Material Instance Dynamic Object Reference)
  ]}
 SetDisableState {nodes [
   n1 = SetDisableState!() -> (then, IsDisabled:bool=~) {is=title K2Node_FunctionEntry}
   > SetScalarParameterValue(
    self=Get DynamicBackground() -> (DynamicBackground:Material Instance Dynamic Object Reference)
    ParameterName=IsDisabled
    Value=Select(Option 0:float=0, Option 1:float=1, Index=@n1) -> (ReturnValue:float)
   )
   n2 = Select(Option 0=1, Option 1=0.6, Index=@n1)
   > Set RenderOpacity!(execute, RenderOpacity:float=@n2, self:Widget Object Reference=Get SettingEntryInfo()) -> (then, Output_Get:float)
   > Set RenderOpacity(RenderOpacity=@n2, self=Get NameSlot_SettingType() -> (NameSlot_SettingType:Named Slot Object Reference))
  ]}
 SetMobileStyle {nodes [
   SetMobileStyle() {is=title K2Node_FunctionEntry}
   > Set Size!(
    execute
    Size:Vector 2D Structure=Select(Option 0:Vector 2D Structure=(X=600,Y=64), Option 1:Vector 2D Structure=(X=600,Y=80), Index=IsUsingCompactUI()) -> (ReturnValue:Vector 2D Structure)
    self:Spacer Object Reference=Get MinSize() -> (MinSize:Spacer Object Reference)
   ) -> (then, Output_Get:Vector 2D Structure)
  ]}
]
variables [DynamicBackground {type=MaterialInstanceDynamic}]
widgets {
 info {inheritedWidgetCount=1, parentClass=/Script/FortniteUI.UIKitModularButton, rootWidgetClass=&CommonBorder, widgetCount=8}
 widgets [
  B_Setting {
   widgetClassPath=&CommonBorder
   bIsVariable=true
   widget=B_Setting
   properties {
    background {
     bIsDynamicallyLoaded=false
     drawAs=Image
     imageType=FullColor
     mirroring=NoMirror
     resourceName=None
     resourceObject=/MatchmakingSettingsUI/MatchmakingSettings/MaterialInstances/MI_UI_MatchmakingSettings_Entry_Background
     tiling=NoTile
     imageSize=<32 32>
     margin=<ltrb 0 0 0 0>
     outlineSettings {
      bUseBrushTransparency=false
      roundingType=HalfHeightRadius
      width=0
      color {colorUseRule=UseColor_Specified, specifiedColor=<rgba 0 0 0 0>}
      cornerRadii=<0 0 0 0>
     }
     tintColor {colorUseRule=UseColor_Specified, specifiedColor=<rgba 1 1 1 1>}
     uVRegion {bIsValid=false, max=<0 0>, min=<0 0>}
    }
    padding=<ltrb 0 0 0 0>
   }
  }
  O_SettingContainer {
   widgetClassPath=&Overlay
   parent=B_Setting
   widget=O_SettingContainer
   properties {bOverride_Cursor=true}
   slot {class=BorderSlot, properties {padding=<ltrb 0 0 0 0>}}
  }
  MinSize {
   widgetClassPath=&Spacer
   bIsVariable=true
   parent=O_SettingContainer
   widget=MinSize
   properties {size=<600 64>}
   slot {class=OverlaySlot, properties {horizontalAlignment=HAlign_Fill, verticalAlignment=VAlign_Fill}}
  }
  HB_SettingEntry {
   widgetClassPath=&HorizontalBox
   parent=O_SettingContainer
   widget=HB_SettingEntry
   slot {class=OverlaySlot, properties {horizontalAlignment=HAlign_Fill, verticalAlignment=VAlign_Fill}}
  }
  SettingEntryInfo {
   widgetClassPath=&/WBP_MatchmakingSettings_SettingEntryInfo
   bIsVariable=true
   parent=HB_SettingEntry
   widget=SettingEntryInfo
   slot {class=HorizontalBoxSlot, properties {verticalAlignment=VAlign_Center, padding=<ltrb 12 8 0 8>, size {sizeRule=Fill, value=1}}}
  }
  NameSlot_SettingType {
   widgetClassPath=&NamedSlot
   bIsVariable=true
   parent=HB_SettingEntry
   widget=NameSlot_SettingType
   slot {class=HorizontalBoxSlot, properties {verticalAlignment=VAlign_Center, padding=<ltrb 24 0 16 0>}}
  }
  WBP_UIKit_Block_Outline {
   widgetClassPath=/FortUILibrary/UIKit/Buttons/Blocks/WBP_UIKit_Block_Outline
   bIsVariable=true
   parent=O_SettingContainer
   widget=WBP_UIKit_Block_Outline
   properties {bottomLeftCornerRadius=20, bottomRightCornerRadius=20, radiusInPixels=true, topLeftCornerRadius=20, topRightCornerRadius=20}
   slot {class=OverlaySlot, properties {horizontalAlignment=HAlign_Fill, verticalAlignment=VAlign_Fill}}
  }
  InputActionWidget {widgetClassPath=&CommonActionWidget, bInherited=true, widget=null}
 ]
}
